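# Sound cue for the TameFoot NPC boss's voice line at the end of an ability, filed under the BigBall ability.
# A Modulator at default settings varies the output of an equal-weight Random node holding three wave variants and two empty inputs, so some triggers can be silent.
# Playback uses the NPC VO sound class, close-range TameFoot attenuation, a chatter concurrency group and boosted volume.
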SoundCue /NPCLibrary/Abilities/BigBall/Sounds/VO/TameFoot_Ability_End_Cue
root=Modulator_0
nodes [
 {id=Modulator_0, children [Random_0]}
 {
  id=Random_0
  children [WavePlayer_0,WavePlayer_2,WavePlayer_3,null,null]
  properties {bShouldExcludeFromBranchCulling=true, weights [1,1,1,1,1]}
 }
 {id=WavePlayer_0, properties {soundWaveAssetPtr=/NPCLibrary/Abilities/BigBall/Sounds/VO/Waves/TameFoot_NPCBoss_AbilityEnd_010}}
 {id=WavePlayer_2, properties {soundWaveAssetPtr=/NPCLibrary/Abilities/BigBall/Sounds/VO/Waves/TameFoot_NPCBoss_AbilityEnd_011}}
 {id=WavePlayer_3, properties {soundWaveAssetPtr=/NPCLibrary/Abilities/BigBall/Sounds/VO/Waves/TameFoot_NPCBoss_AbilityEnd_012}}
]
properties {
 attenuationSettings=&/TameFoot_Close_Att
 maxDistance=4000
 soundClassObject=&Fort_VO_NPC
 volumeMultiplier=1.5
 concurrencySet [&/TameFoot_Chatter_Conc]
}
